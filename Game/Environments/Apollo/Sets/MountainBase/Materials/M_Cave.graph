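# Cave material for the MountainBase set that renders rock, sand or blends of both, with static switches picking mesh UV or world-aligned rock mapping and an optional baked normal.
# Vertex colors mask rock detail, rotating-noise caustics and height-blended sand.
# Hard-coded world positions drive height and entrance color gradients and waterline darkening, tying it to one location.
# Quality switches drop detail, caustics and waterline on Low.
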Material /Game/Environments/Apollo/Sets/MountainBase/Materials/M_Cave
graph {
 properties {bUseMaterialAttributes=true, bUsedWithInstancedStaticMeshes=true, pixelDepthOffsetMode=PDOM_Legacy, usedShadingModels=Default Lit}
 nodes [
  n1 = TextureObject() {texture=/Game/Environments/Apollo/Sets/MountainBase/Textures/Cave_D}
  n2 = TextureSample(Tex=@n1)
  n3 = TextureObject() {samplerType=SAMPLERTYPE_Normal, texture=/Game/Environments/Apollo/Sets/MountainBase/Textures/Cave_N}
  n4 = SetMaterialAttributes!(
   MaterialAttributes
   Base Color=@n2.RGB
   Specular=Multiply(A=@n2.RGB, B=ScalarParameter() {defaultValue=0.5, group=Rocks, parameterName=Rocks Spec, sortPriority=3})
   Roughness=LinearInterpolate(
    A=ScalarParameter() {defaultValue=-0.35, parameterName=R1}
    B=ScalarParameter() {defaultValue=4.5, parameterName=R2}
    Alpha=ComponentMask(None=@n2.RGB) {r=true}
   )
   Normal=TextureSample(Tex=@n3) {samplerType=SAMPLERTYPE_Normal}.RGB
  ) {attributeSetTypes [
    #1
    #2
    #3
    #4
   ]}
  n5 = MaterialFunctionCall!(
   Diffuse Texture Object=@n1
   Spec Texture Object
   TextureSize=Multiply(A=ScalarParameter() {defaultValue=2, group=Rocks, parameterName=ScaleRocks}, B=Constant() {r=1500})
   --------------------------
   Use seperate settings for Z projection
   Use seperate texture for Z projection
   Texture Size Z projection
   Z Plane Diffuse Texture Object
   Z Plane Spec Texture Object
   "--------------------------      "
   ProjectionTransitionContrast
   Use Texture Alpha Heightmap to Refine the Transistion
   HeightLerpTransitionContrast
   PickTheDiffuseHeightChannel
   TheDiffuseHeightChannel
   "--------------------------       "
   NormalTextureObject=@n3
   World Space Normals=StaticBool()
   "--------------------------  "
   Export Float 4 Diffuse=StaticBool() {value=true}
   Export Float 4 Spec
   "--------------------------     "
   WorldPosition
   World Space Normal
  ) {materialFunction=/Engine/Functions/Engine_MaterialFunctions02/Texturing/WorldAlignedTextures_Complex}
  n6 = ComponentMask(None=@n5.XYZ Diffuse Texture) {b=true, g=true, r=true}
  n7 = Multiply(A=@n6, B=ScalarParameter() {defaultValue=0.5, group=Rocks, parameterName=Rocks Spec, sortPriority=3})
  n8 = LinearInterpolate(
   A=ScalarParameter() {defaultValue=-0.35, parameterName=R1}
   B=ScalarParameter() {defaultValue=4.5, parameterName=R2}
   Alpha=ComponentMask(None=@n6) {r=true}
  )
  n9 = GetMaterialAttributes(None=StaticSwitchParameter(
    True=@n4
    False=MakeMaterialAttributes!(
     BaseColor=@n6
     Metallic
     Specular=@n7
     Roughness=@n8
     Anisotropy
     EmissiveColor
     Opacity
     OpacityMask
     Normal=@n5.Normal XYZ Texture
     Tangent
     WorldPositionOffset
     SubsurfaceColor
     ClearCoat
     ClearCoatRoughness
     AmbientOcclusion
     Refraction
     CustomizedUVs_0
     CustomizedUVs_1
     CustomizedUVs_2
     CustomizedUVs_3
     CustomizedUVs_4
     CustomizedUVs_5
     CustomizedUVs_6
     CustomizedUVs_7
     PixelDepthOffset
     ShadingModel
     Displacement
    )
   ) {group=RockFace, parameterName=RockFace}) {attributeGetTypes [#1, #4]}
  n10 = MaterialFunctionCall(
   Diffuse Texture Object=TextureObject() {texture=/Game/Environments/Apollo/Rocks/Natural_Rocks/Textures/T_RockDetails_02D}
   TextureSize=ScalarParameter() {defaultValue=1111, group=Detail Texture, parameterName=WSDetailTile, sortPriority=3}
   ProjectionTransitionContrast=ScalarParameter() {defaultValue=12, parameterName=fdf}
   NormalTextureObject=TextureObject() {samplerType=SAMPLERTYPE_Normal, texture=/Game/Environments/Apollo/Rocks/Natural_Rocks/Textures/T_RockDetails_02_N}
   World Space Normals=StaticBool()
  ) {materialFunction=&WorldAlignedTextures_Complex}
  n11 = ComponentMask(None=@n10.XYZ Diffuse Texture) {r=true}
  n12 = Multiply(A=@n10.Normal XYZ Texture, B=VectorParameter() {parameterName=Details N Value, defaultValue=<rgba 1 1 0 1>}.RGB)
  n13 = BlendMaterialAttributes(
   A=@n9.MaterialAttributes
   B=SetMaterialAttributes!(MaterialAttributes=@n9.MaterialAttributes, Base Color=Multiply(A=@n9.Base Color, B=@n11), Normal=Add(A=@n9.Normal, B=@n12)) {attributeSetTypes [#1, #4]}
   Alpha=VertexColor().R
  )
  n14 = GetMaterialAttributes(None=QualitySwitch(Default=@n13, Low=@n9.MaterialAttributes, High=@n13, Medium=@n13)) {attributeGetTypes [#1]}
  n15 = ScalarParameter() {defaultValue=0.65, parameterName=SideCaves}
  n16 = VectorParameter() {group=Ramp, parameterName=Bottom Cave Color, sortPriority=1, defaultValue=<rgba 0.3333 0.2861 0.2465 0>}
  n17 = VectorParameter() {group=Ramp, parameterName=Top Cave Color, sortPriority=2, defaultValue=<rgba 0.1389 0.4543 0.5104 0>}
  n18 = Subtract(A=WorldPosition().XYZ, B=Constant3Vector() {constant=<rgba -16896 112896 0 0>})
  n19 = ComponentMask(None=@n18) {g=true, r=true}
  n20 = ScalarParameter() {defaultValue=2143.0876, group=Caustics, parameterName=CausticsScale}
  n21 = TextureSampleParameter2D(UVs=Rotator(Coordinate=Divide(A=@n19, B=@n20)) {speed=0.125}) {group=Caustics, parameterName=Lighting, samplerType=SAMPLERTYPE_Grayscale, texture=&LowResBlurredNoise}
  n22 = ComponentMask(None=Add(A=@n18, B=Constant3Vector() {constant=<rgba 555 5555 0 0>})) {g=true, r=true}
  n23 = TextureSampleParameter2D(UVs=Rotator(Coordinate=Divide(A=@n22, B=@n20)) {speed=-0.075}) {group=Caustics, parameterName=Lighting, samplerType=SAMPLERTYPE_Grayscale, texture=&LowResBlurredNoise}
  n24 = Add(A=@n17.RGB, B=LinearInterpolate(B=Multiply(A=@n21.R, B=@n23.R), Alpha=VertexColor().G) {constB=0})
  n25 = QualitySwitch(Default=@n24, Low=@n17.RGB, High=@n24, Medium=@n24)
  n26 = WorldPosition()
  n27 = ComponentMask(None=@n26.XYZ) {b=true}
  n28 = Constant() {r=-3900}
  n29 = Saturate(None=Divide(A=Subtract(A=@n27, B=@n28), B=ScalarParameter() {defaultValue=2237.5549, group=Ramp, parameterName=Ramp Height}))
  n30 = Divide(
   A=Add(A=@n27, B=ScalarParameter() {defaultValue=1653.1692, parameterName=TopHole Height})
   B=ScalarParameter() {defaultValue=2009.1305, parameterName=TopHole Blr}
  )
  n31 = Add(A=ComponentMask(None=@n26.XYZ) {g=true}, B=ScalarParameter() {defaultValue=-108004.1875, parameterName=Side Fade})
  n32 = ScalarParameter() {defaultValue=3555, parameterName=Side Fade Blr}
  n33 = LinearInterpolate(
   A=@n15
   B=LinearInterpolate(
    A=LinearInterpolate(A=@n16.RGB, B=@n25, Alpha=@n29)
    B=ScalarParameter() {defaultValue=1.5, parameterName=Top Light}
    Alpha=Saturate(None=@n30)
   )
   Alpha=Saturate(None=Divide(A=@n31, B=@n32))
  ) {constA=1}
  n34 = ComponentMask(None=@n26.XYZ) {g=true}
  n35 = ScalarParameter() {defaultValue=-121856, parameterName=Enter Fade}
  n36 = Saturate(None=Divide(A=Add(A=@n34, B=@n35), B=ScalarParameter() {defaultValue=-1555, parameterName=Enter Fade Blr}))
  n37 = SetMaterialAttributes!(MaterialAttributes=@n14.MaterialAttributes, Base Color=Multiply(A=@n14.Base Color, B=LinearInterpolate(A=@n15, B=@n33, Alpha=@n36))) {attributeSetTypes [#1]}
  n38 = GetMaterialAttributes(None=@n37) {attributeGetTypes [#1, #3, #2]}
  n39 = Subtract(A=ComponentMask(None=WorldPosition().XYZ) {b=true}, B=Constant() {r=-3900})
  n40 = Saturate(None=Divide(A=@n39, B=ScalarParameter() {defaultValue=300, group=Waterline, parameterName=Waterline Height, sortPriority=3}))
  n41 = LinearInterpolate(A=ScalarParameter() {defaultValue=0.5, group=Waterline, parameterName=Waterline Value}, Alpha=@n40) {constA=0.15}
  n42 = ScalarParameter() {defaultValue=0.2, group=Waterline, parameterName=Waterline Roughness, sortPriority=1}
  n43 = Multiply(A=@n38.Specular, B=ScalarParameter() {defaultValue=0.5, group=Waterline, parameterName=Waterline Spec, sortPriority=1})
  n44 = StaticSwitchParameter(
   True=SetMaterialAttributes!(
    MaterialAttributes=@n38.MaterialAttributes
    Base Color=Multiply(A=@n38.Base Color, B=@n41)
    Roughness=LinearInterpolate(A=@n42, B=@n38.Roughness, Alpha=@n40) {constA=0.5}
    Specular=LinearInterpolate(A=@n43, B=@n38.Specular, Alpha=@n40)
   ) {attributeSetTypes [#1, #3, #2]}
   False=@n38.MaterialAttributes
  ) {defaultValue=true, group=Interior, parameterName="Waterline "}
  n45 = QualitySwitch(Default=@n44, Low=@n37, High=@n44, Medium=@n44)
  n46 = GetMaterialAttributes(None=@n45) {attributeGetTypes [#4]}
  n47 = TextureSampleParameter2D() {
   group=Baked Normal
   parameterName=Normals
   samplerType=SAMPLERTYPE_Normal
   sortPriority=1
   texture=&FlatNormalBlue
  }
  n48 = Add(
   A=@n47.RGB
   B=Multiply(A=@n46.Normal, B=VectorParameter() {group=Baked Normal, parameterName=TileNomalsValue, sortPriority=2, defaultValue=<rgba 3 3 0 1>}.RGB)
  )
  n49 = VectorParameter() {group=Sand, parameterName=SandColor, sortPriority=2, defaultValue=<rgba 0.0833 0.0833 0.0833 0.498>}
  n50 = Saturate(None=Divide(A=@n39, B=ScalarParameter() {defaultValue=32, parameterName=Sand Water Height}))
  n51 = LinearInterpolate(A=Multiply(A=@n49.RGB) {constB=0.35}, B=@n49.RGB, Alpha=@n50)
  n52 = Multiply(A=Abs(None=ScalarParameter() {defaultValue=812.1213, group=Sand, parameterName=Sand WS Scale, sortPriority=1})) {constA=-1, constB=-1}
  n53 = Divide(A=ComponentMask(None=WorldPosition().XYZ) {g=true, r=true}, B=@n52)
  n54 = TextureSample(UVs=@n53, Tex=TextureObject() {samplerType=SAMPLERTYPE_Masks, texture=/Game/Environments/Apollo/Sets/Terrain/Textures/T_Apollo_Sand_S_01}) {samplerType=SAMPLERTYPE_Masks, texture=&T_Apollo_Sand_S_01}
  n55 = LinearInterpolate(B=@n54.B, Alpha=@n50) {constA=0.2}
  n56 = Multiply(A=@n49.A, B=@n54.R)
  n57 = LinearInterpolate(B=@n56, Alpha=@n50) {constA=0.2}
  n58 = TextureObject() {samplerType=SAMPLERTYPE_Normal, texture=/Game/Environments/Apollo/Sets/Terrain/Textures/T_Apollo_Sand_N_01}
  n59 = Transform(None=MaterialFunctionCall!(Negative World Position Divisor=@n52, World Position, Texture=@n58, Z lerp Alpha, WorldNormal, X lerp alpha) {materialFunction=/Engine/Functions/Engine_MaterialFunctions02/WorldAlignedNormals_HighQuality}.Z Projection World Space Normals) {transformSourceType=TRANSFORMSOURCE_World, transformType=TRANSFORM_Tangent}
  n60 = LinearInterpolate(A=Constant3Vector() {constant=<rgba 0 0 1 0>}, B=@n59, Alpha=@n50)
  n61 = SetMaterialAttributes!(MaterialAttributes, Base Color=@n51, Roughness=@n55, Specular=@n57, Normal=@n60) {attributeSetTypes [
    #1
    #3
    #2
    #4
   ]}
  n62 = SetMaterialAttributes(Base Color=@n49.RGB, Roughness=@n54.B, Specular=@n56, Normal=@n59) {attributeSetTypes [
    #1
    #3
    #2
    #4
   ]}
  n63 = QualitySwitch(Default=@n61, Low=@n62, High=@n61, Medium=@n61)
  n64 = VertexColor()
  n65 = MaterialFunctionCall(In=ComponentMask(None=VertexNormalWS()) {b=true}, Contrast=ScalarParameter() {defaultValue=0.4, parameterName=SandOfs}) {materialFunction=&CheapContrast}
  n66 = TextureSample(UVs=@n53) {texture=/Game/Environments/Apollo/Sets/Terrain/Textures/T_Apollo_Sand_D_01}
  n67 = BlendMaterialAttributes(
   A=@n45
   B=@n63
   Alpha=MaterialFunctionCall(Transition Phase=Multiply(A=@n64.B, B=@n65.Result), Height Texture=@n66.A, Contrast=Constant() {r=0.5}) {materialFunction=&HeightLerp}.Alpha
  )
  n68 = StaticSwitchParameter(True=@n67, False=@n45) {group=RockFace, parameterName=RockFaceWithSand}
  n69 = StaticSwitchParameter(True=@n45, False=@n67) {group=RockFace, parameterName=NoSand}
  n70 = Add(
   A=Multiply(A=QualitySwitch(Default=@n60, Low=@n59, High=@n60, Medium=@n60), B=Constant3Vector() {constant=<rgba 0.5 0.5 0 0>})
   B=Multiply(A=TextureSample(Tex=@n58) {samplerType=SAMPLERTYPE_Normal}.RGB, B=Constant3Vector() {constant=<rgba 2 2 1 0>})
  )
  n71 = SetMaterialAttributes(Base Color=@n51, Roughness=@n55, Specular=@n57, Normal=@n70) {attributeSetTypes [
    #1
    #3
    #2
    #4
   ]}
  n72 = SetMaterialAttributes(Base Color=@n49.RGB, Roughness=@n54.B, Specular=@n56, Normal=@n70) {attributeSetTypes [
    #1
    #3
    #2
    #4
   ]}
  n73 = VertexColor()
  StaticSwitchParameter(
   True=BlendMaterialAttributes(
    A=@n63
    B=QualitySwitch(Default=@n71, Low=@n72, High=@n71, Medium=@n71)
    Alpha=MaterialFunctionCall(Transition Phase=@n73.A, Height Texture=@n66.A) {materialFunction=&HeightLerp}.Lerp Alpha No Contrast
   ) {vertexAttributeBlendType=UseA}
   False=StaticSwitchParameter(
    True=SetMaterialAttributes!(MaterialAttributes=@n46.MaterialAttributes, Normal=@n48) {attributeSetTypes [#4]}
    False=StaticSwitchParameter(True=@n68, False=@n69) {group=RockFace, parameterName=RockFace}
   ) {parameterName=Baked Normal}
  ) {group=Sand, parameterName=SandMesh}
 ]
}
