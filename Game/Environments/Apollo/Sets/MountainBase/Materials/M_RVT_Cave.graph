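# Cave rock material for the Apollo MountainBase set that can blend into the landscape runtime virtual texture by vertex alpha.
# WS Rocks switches between world-aligned rock with a height-graded tint masked by vertex blue, and UV-mapped rock topped with height-blended sand on up-facing normals.
# Vertex red blends in a world-aligned detail texture.
# Forward, mobile and material proxy paths fall back to plain world-aligned rock, without detail or sand.
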Material /Game/Environments/Apollo/Sets/MountainBase/Materials/M_RVT_Cave
graph {
 properties {
  bTangentSpaceNormal=false
  bUseMaterialAttributes=true
  bUsedWithInstancedStaticMeshes=true
  pixelDepthOffsetMode=PDOM_Legacy
  usedShadingModels=Default Lit
 }
 nodes [
  n1 = RuntimeVirtualTextureSample!(UVs, World Position) {materialType=BaseColor_Normal_Specular_YCoCg, virtualTexture=/Game/Athena/Apollo/Maps/Landscape/VT/RVT_Apollo_Landscape}
  n2 = SetMaterialAttributes!(MaterialAttributes, Base Color=@n1.BaseColor, Specular=@n1.BaseColor, Roughness=@n1.BaseColor, Normal=@n1.BaseColor) {attributeSetTypes [
    #1
    #2
    #3
    #4
   ]}
  n3 = VertexColor()
  n4 = VectorParameter() {group=Sand, parameterName=SandColor, sortPriority=2, defaultValue=<rgba 0.0833 0.0833 0.0833 0.498>}
  n5 = Multiply(A=Abs(None=ScalarParameter() {defaultValue=812.1213, group=Sand, parameterName=Sand WS Scale, sortPriority=1})) {constA=-1, constB=-1}
  n6 = Divide(A=ComponentMask(None=WorldPosition().XYZ) {g=true, r=true}, B=@n5)
  n7 = TextureSample(UVs=@n6, Tex=TextureObject() {samplerType=SAMPLERTYPE_Masks, texture=/Game/Environments/Apollo/Sets/Terrain/Textures/T_Apollo_Sand_S_01}) {samplerType=SAMPLERTYPE_Masks, texture=&T_Apollo_Sand_S_01}
  n8 = SetMaterialAttributes!(
   MaterialAttributes
   Base Color=@n4.RGB
   Roughness=@n7.B
   Specular=Multiply(A=@n4.A, B=@n7.R)
   Normal=MaterialFunctionCall!(
    Negative World Position Divisor=@n5
    World Position
    Texture=TextureObject() {samplerType=SAMPLERTYPE_Normal, texture=/Game/Environments/Apollo/Sets/Terrain/Textures/T_Apollo_Sand_N_01}
    Z lerp Alpha
    WorldNormal
    X lerp alpha
   ) {materialFunction=/Engine/Functions/Engine_MaterialFunctions02/WorldAlignedNormals_HighQuality}.Z Projection World Space Normals
  ) {attributeSetTypes [
    #1
    #3
    #2
    #4
   ]}
  n9 = MaterialFunctionCall(In=ComponentMask(None=VertexNormalWS()) {b=true}, Contrast=Constant() {r=-0.1576}) {materialFunction=&CheapContrast}
  n10 = OneMinus(None=TextureSample(UVs=@n6) {texture=/Game/Environments/Apollo/Sets/Terrain/Textures/T_Apollo_Sand_D_01}.A)
  n11 = Constant() {r=1}
  n12 = MaterialFunctionCall!(
   Diffuse Texture Object=TextureObject() {texture=/Game/Environments/Apollo/Sets/MountainBase/Textures/Cave_D}
   Spec Texture Object
   TextureSize=Multiply(A=ScalarParameter() {defaultValue=2, group=Rocks, parameterName=ScaleRocks}, B=Constant() {r=1500})
   --------------------------
   Use seperate settings for Z projection
   Use seperate texture for Z projection
   Texture Size Z projection
   Z Plane Diffuse Texture Object
   Z Plane Spec Texture Object
   "--------------------------      "
   ProjectionTransitionContrast
   Use Texture Alpha Heightmap to Refine the Transistion
   HeightLerpTransitionContrast
   PickTheDiffuseHeightChannel
   TheDiffuseHeightChannel
   "--------------------------       "
   NormalTextureObject=TextureObject() {samplerType=SAMPLERTYPE_Normal, texture=/Game/Environments/Apollo/Sets/MountainBase/Textures/Cave_N}
   World Space Normals=StaticBool()
   "--------------------------  "
   Export Float 4 Diffuse=StaticBool() {value=true}
   Export Float 4 Spec
   "--------------------------     "
   WorldPosition
   World Space Normal
  ) {materialFunction=/Engine/Functions/Engine_MaterialFunctions02/Texturing/WorldAlignedTextures_Complex}
  n13 = ComponentMask(None=@n12.XYZ Diffuse Texture) {b=true, g=true, r=true}
  n14 = MaterialFunctionCall(Normal=@n12.Normal XYZ Texture, Flatness=Constant() {r=0.7}) {materialFunction=&FlattenNormal}
  n15 = MakeMaterialAttributes!(
   BaseColor=Multiply(A=VectorParameter() {group=Rocks, parameterName=Rocks Tone, defaultValue=<rgba 1 1 1 0>}.RGB, B=@n13)
   Metallic
   Specular=Multiply(A=@n13, B=ScalarParameter() {defaultValue=0.5, group=Rocks, parameterName=Rocks Spec, sortPriority=3})
   Roughness=LinearInterpolate(
    A=ScalarParameter() {defaultValue=-0.35, parameterName=R1}
    B=ScalarParameter() {defaultValue=4.5, parameterName=R2}
    Alpha=ComponentMask(None=@n13) {r=true}
   )
   Anisotropy
   EmissiveColor
   Opacity
   OpacityMask
   Normal=MaterialProxyReplace(
    Realtime=ShadingPathSwitch(Default=@n14.Result, Deferred=@n14.Result, Forward=@n12.Normal XYZ Texture, Mobile=@n12.Normal XYZ Texture)
    MaterialProxy=@n12.Normal XYZ Texture
   )
   Tangent
   WorldPositionOffset
   SubsurfaceColor
   ClearCoat
   ClearCoatRoughness
   AmbientOcclusion
   Refraction
   CustomizedUVs_0
   CustomizedUVs_1
   CustomizedUVs_2
   CustomizedUVs_3
   CustomizedUVs_4
   CustomizedUVs_5
   CustomizedUVs_6
   CustomizedUVs_7
   PixelDepthOffset
   ShadingModel
   Displacement
  )
  n16 = GetMaterialAttributes(None=@n15) {attributeGetTypes [#1]}
  n17 = VectorParameter() {group=Ramp, parameterName=Top Cave Color, sortPriority=2, defaultValue=<rgba 0.1389 0.4543 0.5104 0>}
  n18 = ScalarParameter() {defaultValue=1.5, parameterName=Top Light}
  n19 = ComponentMask(None=WorldPosition().XYZ) {b=true}
  n20 = ScalarParameter() {defaultValue=1653.1692, parameterName=TopHole Height}
  n21 = Saturate(None=Divide(A=Add(A=@n19, B=@n20), B=ScalarParameter() {defaultValue=2009.1305, parameterName=TopHole Blr}))
  n22 = LinearInterpolate(A=@n16.Base Color, B=Multiply(A=@n16.Base Color, B=LinearInterpolate(A=@n17.RGB, B=@n18, Alpha=@n21)), Alpha=VertexColor().B)
  n23 = TextureSample(Tex=TextureObject() {texture=&Cave_D})
  n24 = TextureObject() {samplerType=SAMPLERTYPE_Normal, texture=&Cave_N}
  n25 = VertexColor()
  n26 = SetMaterialAttributes!(
   MaterialAttributes
   Base Color=Multiply(A=@n23.RGB, B=VectorParameter() {group=Rocks, parameterName=Rocks Tone, defaultValue=<rgba 0.7344 0.7344 0.7344 0>}.RGB)
   Specular=Multiply(A=@n23.RGB, B=ScalarParameter() {defaultValue=0.5, group=Rocks, parameterName=Rocks Spec, sortPriority=3})
   Roughness=LinearInterpolate(
    A=ScalarParameter() {defaultValue=-0.35, parameterName=R1}
    B=ScalarParameter() {defaultValue=4.5, parameterName=R2}
    Alpha=ComponentMask(None=@n23.RGB) {r=true}
   )
   Normal=MaterialFunctionCall(Normal=TextureSample(Tex=@n24) {samplerType=SAMPLERTYPE_Normal}.RGB, Flatness=OneMinus(None=@n25.B)) {materialFunction=&FlattenNormal}.Result
  ) {attributeSetTypes [
    #1
    #2
    #3
    #4
   ]}
  n27 = GetMaterialAttributes(None=StaticSwitchParameter(
    True=SetMaterialAttributes!(MaterialAttributes=@n16.MaterialAttributes, Base Color=@n22) {attributeSetTypes [#1]}
    False=@n26
   ) {parameterName=WS Rocks}) {attributeGetTypes [#1, #4]}
  n28 = MaterialFunctionCall(
   Diffuse Texture Object=TextureObject() {texture=/Game/Environments/Apollo/Rocks/Natural_Rocks/Textures/T_RockDetails_02D}
   TextureSize=ScalarParameter() {defaultValue=1111, group=Detail Texture, parameterName=WSDetailTile, sortPriority=3}
   ProjectionTransitionContrast=Constant() {r=12}
   NormalTextureObject=TextureObject() {samplerType=SAMPLERTYPE_Normal, texture=/Game/Environments/Apollo/Rocks/Natural_Rocks/Textures/T_RockDetails_02_N}
   World Space Normals=StaticBool()
  ) {materialFunction=&WorldAlignedTextures_Complex}
  n29 = ComponentMask(None=@n28.XYZ Diffuse Texture) {r=true}
  n30 = Multiply(A=@n28.Normal XYZ Texture, B=VectorParameter() {parameterName=Details N Value, defaultValue=<rgba 1 1 0 1>}.RGB)
  n31 = BlendMaterialAttributes(
   A=@n27.MaterialAttributes
   B=SetMaterialAttributes!(MaterialAttributes=@n27.MaterialAttributes, Base Color=Multiply(A=@n27.Base Color, B=@n29), Normal=Add(A=@n27.Normal, B=@n30)) {attributeSetTypes [#1, #4]}
   Alpha=VertexColor().R
  )
  n32 = GetMaterialAttributes(None=ShadingPathSwitch(Default=@n31, Deferred=@n31, Forward=@n15, Mobile=@n15)) {attributeGetTypes [#4]}
  n33 = SetMaterialAttributes!(MaterialAttributes=@n32.MaterialAttributes, Normal=Transform(None=@n32.Normal)) {attributeSetTypes [#4]}
  n34 = StaticSwitchParameter(
   True=@n33
   False=BlendMaterialAttributes(
    A=@n33
    B=@n8
    Alpha=MaterialFunctionCall(Transition Phase=@n9.Result, Height Texture=@n10, Contrast=@n11) {materialFunction=&HeightLerp}.Alpha
   )
  ) {parameterName=WS Rocks}
  n35 = ShadingPathSwitch(Default=@n34, Deferred=@n34, Forward=@n33, Mobile=@n33)
  MaterialProxyReplace(
   Realtime=MaterialFunctionCall(NoRVT=@n35, RVT=BlendMaterialAttributes(A=@n2, B=@n35, Alpha=@n3.A)) {materialFunction=&FN_RVT_Switch_Material}.Result
   MaterialProxy=@n15
  )
 ]
}
