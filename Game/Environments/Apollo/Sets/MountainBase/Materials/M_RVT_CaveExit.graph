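# Opaque material for the Apollo MountainBase cave exit that blends rock, sand and landscape runtime virtual texture.
# Cave rock, world-aligned or mesh-UV via RockFace switch, gets brightness gradients from world Y/Z to fake opening light.
# Vertex red adds world-aligned detail; blue times upward-facing normal height-blends in world-projected sand.
# Vertex alpha height-blends toward the landscape RVT when virtual texturing is available; material proxies skip that blend.
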Material /Game/Environments/Apollo/Sets/MountainBase/Materials/M_RVT_CaveExit
graph {
 properties {bTangentSpaceNormal=false, bUseMaterialAttributes=true, pixelDepthOffsetMode=PDOM_Legacy, usedShadingModels=Default Lit}
 nodes [
  n1 = RuntimeVirtualTextureSample!(UVs, World Position) {materialType=BaseColor_Normal_Specular_YCoCg, virtualTexture=/Game/Athena/Apollo/RVT/RVT_Apollo_Landscape}
  n2 = SetMaterialAttributes!(MaterialAttributes, Base Color=@n1.BaseColor, Specular=@n1.BaseColor, Roughness=@n1.BaseColor, Normal=@n1.BaseColor) {attributeSetTypes [
    #1
    #2
    #3
    #4
   ]}
  n3 = Multiply(A=Abs(None=ScalarParameter() {defaultValue=812.1213, group=Sand, parameterName=Sand WS Scale, sortPriority=1})) {constA=-1, constB=-1}
  n4 = Divide(A=ComponentMask(None=WorldPosition().XYZ) {g=true, r=true}, B=@n3)
  n5 = TextureSample(UVs=@n4) {texture=/Game/Environments/Apollo/Sets/Terrain/Textures/T_Apollo_Sand_D_01}
  n6 = MaterialFunctionCall(Transition Phase=VertexColor().A, Height Texture=@n5.A, Contrast=Constant() {r=1}) {materialFunction=&HeightLerp}
  n7 = VectorParameter() {group=Sand, parameterName=SandColor, sortPriority=2, defaultValue=<rgba 0.0833 0.0833 0.0833 0.498>}
  n8 = TextureSample(UVs=@n4, Tex=TextureObject() {samplerType=SAMPLERTYPE_Masks, texture=/Game/Environments/Apollo/Sets/Terrain/Textures/T_Apollo_Sand_S_01}) {samplerType=SAMPLERTYPE_Masks, texture=&T_Apollo_Sand_S_01}
  n9 = SetMaterialAttributes!(
   MaterialAttributes
   Base Color=@n7.RGB
   Roughness=@n8.B
   Specular=Multiply(A=@n7.A, B=@n8.R)
   Normal=MaterialFunctionCall!(
    Negative World Position Divisor=@n3
    World Position
    Texture=TextureObject() {samplerType=SAMPLERTYPE_Normal, texture=/Game/Environments/Apollo/Sets/Terrain/Textures/T_Apollo_Sand_N_01}
    Z lerp Alpha
    WorldNormal
    X lerp alpha
   ) {materialFunction=/Engine/Functions/Engine_MaterialFunctions02/WorldAlignedNormals_HighQuality}.Z Projection World Space Normals
  ) {attributeSetTypes [
    #1
    #3
    #2
    #4
   ]}
  n10 = VertexColor()
  n11 = MaterialFunctionCall(In=ComponentMask(None=VertexNormalWS()) {b=true}, Contrast=ScalarParameter() {defaultValue=0.4, parameterName=SandOfs}) {materialFunction=&CheapContrast}
  n12 = TextureObject() {texture=/Game/Environments/Apollo/Sets/MountainBase/Textures/Cave_D}
  n13 = TextureObject() {samplerType=SAMPLERTYPE_Normal, texture=/Game/Environments/Apollo/Sets/MountainBase/Textures/Cave_N}
  n14 = MaterialFunctionCall!(
   Diffuse Texture Object=@n12
   Spec Texture Object
   TextureSize=Multiply(A=ScalarParameter() {defaultValue=2, group=Rocks, parameterName=ScaleRocks}, B=Constant() {r=1500})
   --------------------------
   Use seperate settings for Z projection
   Use seperate texture for Z projection
   Texture Size Z projection
   Z Plane Diffuse Texture Object
   Z Plane Spec Texture Object
   "--------------------------      "
   ProjectionTransitionContrast
   Use Texture Alpha Heightmap to Refine the Transistion
   HeightLerpTransitionContrast
   PickTheDiffuseHeightChannel
   TheDiffuseHeightChannel
   "--------------------------       "
   NormalTextureObject=@n13
   World Space Normals=StaticBool()
   "--------------------------  "
   Export Float 4 Diffuse=StaticBool() {value=true}
   Export Float 4 Spec
   "--------------------------     "
   WorldPosition
   World Space Normal
  ) {materialFunction=/Engine/Functions/Engine_MaterialFunctions02/Texturing/WorldAlignedTextures_Complex}
  n15 = ComponentMask(None=StaticSwitchParameter(True=TextureSample(Tex=@n12).RGB, False=@n14.XYZ Diffuse Texture) {parameterName=RockFace}) {b=true, g=true, r=true}
  n16 = ScalarParameter() {defaultValue=0.5, group=Rocks, parameterName=Rocks Spec, sortPriority=3}
  n17 = ScalarParameter() {defaultValue=-0.35, parameterName=R1}
  n18 = ScalarParameter() {defaultValue=4.5, parameterName=R2}
  n19 = ComponentMask(None=@n15) {r=true}
  n20 = TextureSample(Tex=@n13)
  n21 = GetMaterialAttributes(None=MakeMaterialAttributes!(
    BaseColor=@n15
    Metallic
    Specular=Multiply(A=@n15, B=@n16)
    Roughness=LinearInterpolate(A=@n17, B=@n18, Alpha=@n19)
    Anisotropy
    EmissiveColor
    Opacity
    OpacityMask
    Normal=StaticSwitchParameter(True=@n20.RGB, False=@n14.Normal XYZ Texture) {parameterName=RockFace}
    Tangent
    WorldPositionOffset
    SubsurfaceColor
    ClearCoat
    ClearCoatRoughness
    AmbientOcclusion
    Refraction
    CustomizedUVs_0
    CustomizedUVs_1
    CustomizedUVs_2
    CustomizedUVs_3
    CustomizedUVs_4
    CustomizedUVs_5
    CustomizedUVs_6
    CustomizedUVs_7
    PixelDepthOffset
    ShadingModel
    Displacement
   )) {attributeGetTypes [#1]}
  n22 = ScalarParameter() {defaultValue=0.65, parameterName=SideCaves}
  n23 = WorldPosition()
  n24 = Divide(
   A=Add(A=ComponentMask(None=@n23.XYZ) {b=true}, B=ScalarParameter() {defaultValue=1653.1692, parameterName=TopHole Height})
   B=ScalarParameter() {defaultValue=2009.1305, parameterName=TopHole Blr}
  )
  n25 = Add(A=ComponentMask(None=@n23.XYZ) {g=true}, B=ScalarParameter() {defaultValue=-108004.1875, parameterName=Side Fade})
  n26 = ScalarParameter() {defaultValue=3555, parameterName=Side Fade Blr}
  n27 = LinearInterpolate(
   A=@n22
   B=LinearInterpolate(A=@n22, B=ScalarParameter() {defaultValue=1.5, parameterName=Top Light}, Alpha=Saturate(None=@n24))
   Alpha=Saturate(None=Divide(A=@n25, B=@n26))
  ) {constA=1}
  n28 = GetMaterialAttributes(None=SetMaterialAttributes!(MaterialAttributes=@n21.MaterialAttributes, Base Color=Multiply(A=@n21.Base Color, B=@n27)) {attributeSetTypes [#1]}) {attributeGetTypes [#1, #4]}
  n29 = MaterialFunctionCall(
   Diffuse Texture Object=TextureObject() {texture=/Game/Environments/Apollo/Rocks/Natural_Rocks/Textures/T_RockDetails_02D}
   TextureSize=ScalarParameter() {defaultValue=1111, group=Detail Texture, parameterName=WSDetailTile, sortPriority=3}
   ProjectionTransitionContrast=Constant() {r=12}
   NormalTextureObject=TextureObject() {samplerType=SAMPLERTYPE_Normal, texture=/Game/Environments/Apollo/Rocks/Natural_Rocks/Textures/T_RockDetails_02_N}
   World Space Normals=StaticBool()
  ) {materialFunction=&WorldAlignedTextures_Complex}
  n30 = Multiply(A=@n28.Base Color, B=ComponentMask(None=@n29.XYZ Diffuse Texture) {r=true})
  n31 = Add(A=@n28.Normal, B=Multiply(A=@n29.Normal XYZ Texture, B=VectorParameter() {parameterName=Details N Value, defaultValue=<rgba 1 1 0 1>}.RGB))
  n32 = GetMaterialAttributes(None=BlendMaterialAttributes(
    A=@n28.MaterialAttributes
    B=SetMaterialAttributes!(MaterialAttributes=@n28.MaterialAttributes, Base Color=@n30, Normal=@n31) {attributeSetTypes [#1, #4]}
    Alpha=VertexColor().R
   )) {attributeGetTypes [#4]}
  n33 = BlendMaterialAttributes(
   A=SetMaterialAttributes!(MaterialAttributes=@n32.MaterialAttributes, Normal=Transform(None=@n32.Normal)) {attributeSetTypes [#4]}
   B=@n9
   Alpha=MaterialFunctionCall(Transition Phase=Multiply(A=@n10.B, B=@n11.Result), Height Texture=@n5.A, Contrast=Constant() {r=0.5}) {materialFunction=&HeightLerp}.Alpha
  )
  MaterialProxyReplace(Realtime=VirtualTextureFeatureSwitch!(No=@n33, Yes=BlendMaterialAttributes(A=@n2, B=@n33, Alpha=@n6.Alpha)), MaterialProxy=@n33)
  Normalize(VectorInput=@n31)
  OneMinus()
  VectorParameter() {group=Ramp, parameterName=Bottom Cave Color, sortPriority=1, defaultValue=<rgba 0.1719 0.1319 0.0985 0>}
 ]
}
