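# Default Lit wall material for the Apollo MountainBase set that prints an Alter Ego logo onto tiled concrete.
# Concrete color, normal and mask textures share doubled UVs, with the color multiplied by a WallTint parameter.
# A dot product of the logo mask and vertex color lets painted RGB select logo channels.
# That result blends from white to LogoTint, which multiplies over the tinted concrete.
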Material /Game/Environments/Apollo/Sets/MountainBase/Materials/M_WallPrint
graph {
 properties {pixelDepthOffsetMode=PDOM_Legacy, usedShadingModels=Default Lit}
 nodes [
  n1 = Multiply(A=TextureCoordinate(), B=Constant() {r=2})
  n2 = TextureSample() {samplerType=SAMPLERTYPE_Masks, texture=/Game/Environments/Apollo/Sets/MountainBase/Textures/T_MountainBase_AlterEgo_Logo_D}
  n3 = VertexColor()
  Multiply(
   A=Multiply(
    A=VectorParameter() {parameterName=WallTint, defaultValue=<rgba 0.3542 0.3542 0.3542 0>}.RGB
    B=TextureSample(UVs=@n1) {texture=/Game/Environments/Apollo/Sets/MountainBase/Textures/T_Concrete_Tile_D}.RGB
   )
   B=LinearInterpolate(B=VectorParameter() {parameterName=LogoTint, defaultValue=<rgba 0.1406 0.1406 0.1406 0>}.RGB, Alpha=DotProduct(A=@n2.RGB, B=@n3)) {constA=1}
  )
  TextureSample(UVs=@n1) {samplerType=SAMPLERTYPE_Normal, texture=/Game/Environments/Apollo/Sets/MountainBase/Textures/T_Concrete_Tile_N}
  TextureSample(UVs=@n1) {samplerType=SAMPLERTYPE_Masks, texture=/Game/Environments/Apollo/Sets/MountainBase/Textures/T_Concrete_Tile_S}
 ]
}
